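# Translucent, unlit, two-sided mesh-particle material for an alarm cone effect on the fragment container gadget.
# Particle color RGB feeds a shared emissive adjustment function for the glow color.
# Opacity combines a vertical soft-fade ramp, particle alpha and a view-angle fade with a tunable sensitivity, then passes through depth fade to soften intersections with geometry.
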Material /Game/Gadgets/Assets/Fragment_Container/Material/M_Alarm_Cone
graph {
 properties {
  bUsedWithMeshParticles=true
  bUsedWithNiagaraMeshParticles=true
  blendMode=BLEND_Translucent
  pixelDepthOffsetMode=PDOM_Legacy
  shadingModel=MSM_Unlit
  twoSided=true
  usedShadingModels=Unlit
 }
 nodes [
  n1 = TextureSample() {texture=/Game/Effects/Fort_Effects/Textures/Ramps/T_Ramp_Vert_Soft_Fade}
  n2 = ParticleColor()
  n3 = ScalarParameter() {defaultValue=4, parameterName=FadeSensitivity}
  DepthFade(Opacity=Multiply(
    A=Multiply(A=@n1.G, B=@n2.A)
    B=MaterialFunctionCall("FadeSensitivity (optional)"=@n3) {materialFunction=&FadeBasedOnViewAngle}.Result
   )) {fadeDistanceDefault=25}
  MaterialFunctionCall(Emissive=@n2.RGB) {materialFunction=&MF_FX_EmissiveAdjustments}
 ]
}
