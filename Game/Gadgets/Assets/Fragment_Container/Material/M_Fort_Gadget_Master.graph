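# Translucent, surface-lit master material for Fort gadgets, stored with the Fragment_Container gadget assets and enabled for skeletal meshes and mesh deformers.
# It exposes base color, alpha, roughness, metalness and normal map parameters, defaulting to a glossy, mostly opaque dark gray with a flat normal.
# The graph also calls the shared LODEffect material function, which has distance-cull fade inputs; the file shows no wiring between nodes.
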Material /Game/Gadgets/Assets/Fragment_Container/Material/M_Fort_Gadget_Master
graph {
 properties {
  bUsedWithMeshDeformer=true
  bUsedWithSkeletalMesh=true
  blendMode=BLEND_Translucent
  pixelDepthOffsetMode=PDOM_Legacy
  translucencyLightingMode=TLM_Surface
  usedShadingModels=Default Lit
 }
 nodes [
  MaterialFunctionCall!(World Position, DistanceCullFadeAnimationOffset, Use actor center instead of pivot point) {materialFunction=/Game/Packages/Fortress_SharedMaterials/Base_Material_Functions/LODEffect}
  ScalarParameter() {defaultValue=0.9, parameterName=Alpha}
  ScalarParameter() {defaultValue=0.02, parameterName=Roughness}
  ScalarParameter() {defaultValue=0.1, parameterName=Metalness}
  TextureSampleParameter2D() {parameterName=NormalMap, samplerType=SAMPLERTYPE_Normal, texture=&FlatNormalBlue}
  VectorParameter() {parameterName=BaseColor, defaultValue=<rgba 0.2 0.2 0.2 0>}
 ]
}
