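# M_Fragment_Flare is an unlit translucent particle sprite material for a flare effect, filed under the Fragment Container gadget assets.
# Particle color RGB passes through the day/night auto bloom control, with a daytime bloom multiplier, then into the FX emissive adjustments function.
# The camera flare texture's green channel is multiplied by particle alpha to form a shaped mask.
# A second, round hard flare texture sample is present but unconnected.
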Material /Game/Gadgets/Assets/Fragment_Container/Material/M_Fragment_Flare
graph {
 properties {
  bUsedWithParticleSprites=true
  blendMode=BLEND_Translucent
  pixelDepthOffsetMode=PDOM_Legacy
  shadingModel=MSM_Unlit
  usedShadingModels=Unlit
 }
 nodes [
  n1 = ParticleColor()
  MaterialFunctionCall(Emissive=MaterialFunctionCall!(DayBloom_Mult=Constant() {r=1.5}, In=@n1.RGB) {materialFunction=/Game/Packages/Fortnite_UniqueMaterials_1/MH_Materials/DayNight_Auto_Bloom_Control}.Result) {materialFunction=&MF_FX_EmissiveAdjustments}
  Multiply(A=TextureSample() {samplerType=SAMPLERTYPE_Grayscale, texture=/Game/Effects/Fort_Effects/Textures/T_EFX_CameraFlare_01}.G, B=@n1.A)
  TextureSample() {samplerType=SAMPLERTYPE_Grayscale, texture=/Game/Effects/Fort_Effects/Textures/T_EFX_Flare_Round_Hard_01}
 ]
}
