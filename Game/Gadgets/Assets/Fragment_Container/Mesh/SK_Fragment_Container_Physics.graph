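# Physics asset for the Fragment Container gadget's skeletal mesh, giving it simple collision for physics.
# A single capsule body on the top_section bone wraps the mesh, with a large radius and very short length so it behaves almost like a sphere.
# No constraints are defined, so the body has no jointed or ragdoll articulation.
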PhysicsAsset /Game/Gadgets/Assets/Fragment_Container/Mesh/SK_Fragment_Container_Physics
bodies {top_section [{shapeType=Capsule, length=6.6291, radius=90.9328, center=<0.3965 3.6192 -10.4945>}]}
constraints []
